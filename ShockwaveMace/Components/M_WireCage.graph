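# M_WireCage is a translucent, two-sided material in the ShockwaveMace components that computes a glowing wire-grid cage look.
# Object-relative world position is wrapped by Fmod at the Scale spacing, then squared, inverted and floored at zero to leave thin bands along each axis.
# A dot product with a constant merges the axis bands into one mask.
# Separately, a Color parameter, orange by default, is multiplied by EmissiveBrightness for the glow.
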Material /ShockwaveMace/Components/M_WireCage
graph {
 properties {blendMode=BLEND_Translucent, pixelDepthOffsetMode=PDOM_Legacy, twoSided=true, usedShadingModels=Default Lit}
 nodes [
  n1 = Fmod(A=Subtract(A=WorldPosition().XYZ, B=ObjectPositionWS()), B=ScalarParameter() {defaultValue=30, parameterName=Scale})
  n2 = Multiply(A=@n1, B=@n1)
  DotProduct(A=Max(A=OneMinus(None=@n2)) {constB=0}, B=Constant() {r=1})
  Multiply(
   A=VectorParameter() {parameterName=Color, defaultValue=<rgba 1 0.25 0 1>}.RGB
   B=ScalarParameter() {defaultValue=10, parameterName=EmissiveBrightness}
  )
 ]
}
