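; Pawn component for the Shockwave Mace that cleans up its bounce ability and animation layer.
; When a scalable-float flag is on, the authority watches movement mode changes and cancels Bounce-tagged abilities on Walking-to-Falling, a workaround for unreliable anim-notify removal (FORT-976546).
; UnlinkAnimLayerClient routes from the locally controlled pawn through TellServerUnlinkAnimLayer to the server, which multicasts the layer unlink and applies the bounce-cancel gameplay effect.
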(event EventBeginPlay
  (if (ScalableFloat|GetValueasBool (Variables|Default|GetForceRemoveBounceAbility))
    ;[0.15 9e-05 0.02459 0.5 25] FORT-976546 - Shockwave Bounce ability sometimes isn't getting successfully removed via anim notifies. So this ensures that, when we transition from Walking to Falling (as we do when we reach the apex of the hammer bounce), we remove the lingering ability so it doesn't continue to block tags once the player has landed.
    (bind _asfort_pawn (Utilities|Casting|CastToFortPawn (Components|GetOwner)))
    (Utilities|SwitchHasAuthority
      (:Authority
        (Character|BindEventtoMovementModeChangedDelegate _asfort_pawn (delegate Custom|MoveModeChanged))))))

(event Custom|MoveModeChanged
  (params (Character :type Character%20Object%20Reference) (PrevMovementMode :type EMovementMode%20Enum :default "MOVE_None") (PreviousCustomMode :type Byte))
  ;[0.15 9e-05 0.02459 0.5 25] FORT-976546 - Shockwave Bounce ability sometimes isn't getting successfully removed via anim notifies. So this ensures that, when we transition from Walking to Falling (as we do when we reach the apex of the hammer bounce), we remove the lingering ability so it doesn't continue to block tags once the player has landed.
  (if (Utilities|Enum|Equal(Enum) PrevMovementMode "MOVE_Walking")
    (switch Utilities|FlowControl|Switch|SwitchonEMovementMode (Variables|/Script/Engine.CharacterMovementComponent|GetMovementMode (Variables|/Script/Engine.Character|GetCharacterMovement Character))
      (:MOVE_Falling
        (Tags|CancelAbilitiesWithTags (Ability|GetFortAbilitySystemComponent Character) (GameplayTags|MakeLiteralGameplayTagContainer "(GameplayTags=((TagName=\"Abilities.ShockwaveMace.Bounce\")))"))))))

(event Custom|TellServerUnlinkAnimLayer
  (CallFunction|UnlinkAnimLayeronServer))

(fn UnlinkAnimLayerClient
  (bind _aspawn (Utilities|Casting|CastToPawn (Components|GetOwner)))
  (if (Pawn|IsLocallyControlled _aspawn)
    (CallFunction|TellServerUnlinkAnimLayer)))

(fn UnlinkAnimLayerOnServer
  (if (Networking|IsServer)
    (bind _asfort_pawn (Utilities|Casting|CastToFortPawn (Components|GetOwner)))
    (Animation|MulticastUnlinkAnimLayer _asfort_pawn (Variables|Default|GetShockwaveLayer))
    (GameplayEffects|ApplyGameplayEffectToSelf :self (Ability|GetFortAbilitySystemComponent _asfort_pawn) :GameplayEffectClass "/ShockwaveMace/GameplayEffects/GE_ShockwaveMace_Bounce_Cancel.GE_ShockwaveMace_Bounce_Cancel_C" :Level 0.0)))
