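; Static mesh component for the Shockwave Mace's custom targeting, exposing a small set of custom events that control the mesh.
; Separate events set hidden-in-game state with propagation to child components and move the mesh to a supplied world transform.
; Two further events toggle collision between none and query-only, so the mesh never enables physics collision through them.
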(event Custom|SetMeshHiddenInGame
  (params (NewHidden :type Boolean))
  (Development|SetHiddeninGame NewHidden true))

(event Custom|UpdateMeshTransform
  (params (NewTransform :type Transform%20%28by%20ref%29))
  (Transformation|SetWorldTransform NewTransform))

(event Custom|DisableCollision
  (Collision|SetCollisionEnabled "NoCollision"))

(event Custom|EnableCollision
  (Collision|SetCollisionEnabled "QueryOnly"))
